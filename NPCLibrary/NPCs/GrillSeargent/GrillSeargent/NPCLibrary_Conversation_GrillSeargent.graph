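# Conversation database for the GrillSeargent NPC, defining the dialogue shown when a player talks to it.
# Its entry point opens a Speech node with a default line, overridden by prioritized contextual greetings for a first-ever conversation or specific Pizza and Durr Burger outfits.
# The speech then branches through link nodes into shared Season 16 Spire Joneses, NPC Library quest and basic services conversations.
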ConversationDatabase /NPCLibrary/NPCs/GrillSeargent/GrillSeargent/NPCLibrary_Conversation_GrillSeargent
entry_points [{tag=Conversation.Entry.GrillSeargent, destinations [#1]}]
nodes [
 ConversationEntryPointNode_7 {
  id=#1
  class=ConversationEntryPointNode
  outputs [#2]
  properties {entryTag {tagName=Conversation.Entry.GrillSeargent}}
 }
 Speech_C_4 {
  id=#2
  class=Speech_C
  outputs [#3, #4, #5]
  properties {
   nodeName=""
   generalConfig {
    defaultMessage="Move those buns, <b>soldier.</>"
    contextualMessages [
     {
      message="Welcome to Durr Burger, home of the-- Oh, that's not what you're here for, huh?"
      contextRequirements [{requirement {class=FortControllerRequirement_IsFirstEverConversationWithNPC}}]
      priority {curve {rowName=Tandem.Speech.Priority.FirstEverInteraction}}
     }
     {
      message="Ah! Boss! I promise, I'm just taking my regulated lunchbreak!"
      requirementMatchPolicy=RequireAny
      contextRequirements [{requirement {
         class=FortControllerRequirement_HasCID
         properties {allowedCIds [
           /BRCosmetics/Athena/Items/Cosmetics/Characters/CID_109_Athena_Commando_M_Pizza
           /BRCosmetics/Athena/Items/Cosmetics/Characters/CID_185_Athena_Commando_M_DurrburgerHero
          ]}
        }}]
      priority {curve {rowName=Tandem.Speech.Priority.OutfitReaction}}
     }
    ]
   }
  }
 }
 ConversationLinkNode_0 {id=#5, class=ConversationLinkNode, properties {remoteEntryTag {tagName=Conversation.Entry.BasicServices}}}
 ConversationLinkNode_1 {id=#4, class=ConversationLinkNode, properties {remoteEntryTag {tagName=Conversation.Entry.NPCLibrary.Quests}}}
 ConversationLinkNode_2 {
  id=#3
  class=ConversationLinkNode
  properties {remoteEntryTag {tagName=Conversation.Entry.Season16.TheSpire.Joneses}}
 }
]
